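# DadBroStingerTest_HighPri_Cue is a test sound cue that plays a Storm King negative weak-point-revealed orchestral music stinger.
# A single wave player feeds the output directly, with no randomization or mixing.
# It routes through the boss fight sound class with a raised priority, boosted volume and spatialization turned off.
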SoundCue /Game/Sounds/Fort_Test_Sounds/Test_Map_Sounds/DadBroStingerTest_HighPri_Cue
root=WavePlayer_1
nodes [{
  id=WavePlayer_1
  properties {soundWaveAssetPtr=/Game/Sounds/Fort_Music/Orchestral_Score/Orch_Storm_King/Final_MIxes/StormKing_WeakPoint_Revealed_Negative_02}
 }]
properties {
 maxDistance=2097152
 priority=2
 soundClassObject=&Fort_BossFight
 volumeMultiplier=1.25
 attenuationOverrides {bSpatialize=false}
 concurrencySet [null]
}
