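# Looping quad MetaSound source for outdoor ambience that plays random waves from an array at random start points.
# With multiple waves, it switches on a randomly chosen interval, flipping a bool to alternate sides of a quad crossfade player.
# Per-wave random pitch and volume multiply with slow AD envelopes that self-retrigger after random delays via deferred trigger variables.
# An azimuth panner applies listener orientation and gain before the four outputs.
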MetaSoundSource /FortMetasound/Ambient/MSS_AmbientLoop_Quad
inputs [
 Enveloper.Pitch.MaxAttackTime {type=Time}
 Enveloper.Pitch.MaxReleaseTime {type=Time}
 Enveloper.Pitch.MaxVolumeRange {type=Float}
 Enveloper.Pitch.MinAttackTime {type=Time}
 Enveloper.Pitch.MinReleaseTime {type=Time}
 Enveloper.Pitch.MinVolumeRange {type=Float}
 Enveloper.Volume.MaxAttackTime {type=Time}
 Enveloper.Volume.MaxReleaseTime {type=Time}
 Enveloper.Volume.MaxVolumeRange {type=Float}
 Enveloper.Volume.MinAttackTime {type=Time}
 Enveloper.Volume.MinReleaseTime {type=Time}
 Enveloper.Volume.MinVolumeRange {type=Float}
 MaxWaveSwitchDuration {type=Time}
 MinWaveSwitchDuration {type=Time}
 Pitch.PitchMax {type=Float}
 Pitch.PitchMin {type=Float}
 UE.Listener.Orientation.Azimuth {type=Float}
 UE.Listener.Orientation.Elevation {type=Float}
 UE.Source.OnPlay {type=Trigger}
 Volume.VolumeMax {type=Float}
 Volume.VolumeMin {type=Float}
 Wave.Wave Assets {type=WaveAsset:Array}
]
interfaces [UE.Listener.Orientation:1.0, UE.OutputFormat.Quad:1.0, UE.Source:1.1]
outputs [
 "UE.OutputFormat.Quad.Audio:0" {type=Audio}
 "UE.OutputFormat.Quad.Audio:1" {type=Audio}
 "UE.OutputFormat.Quad.Audio:4" {type=Audio}
 "UE.OutputFormat.Quad.Audio:5" {type=Audio}
]
pages [Default {
  input_defaults {
   Enveloper.Pitch.MaxAttackTime=15
   Enveloper.Pitch.MaxReleaseTime=15
   Enveloper.Pitch.MaxVolumeRange=0.5
   Enveloper.Pitch.MinAttackTime=10
   Enveloper.Pitch.MinReleaseTime=10
   Enveloper.Pitch.MinVolumeRange=-0.5
   Enveloper.Volume.MaxAttackTime=10
   Enveloper.Volume.MaxReleaseTime=15
   Enveloper.Volume.MaxVolumeRange=1.25
   Enveloper.Volume.MinAttackTime=5
   Enveloper.Volume.MinReleaseTime=10
   Enveloper.Volume.MinVolumeRange=0.6
   MaxWaveSwitchDuration=120
   MinWaveSwitchDuration=60
   Pitch.PitchMax=0.5
   Pitch.PitchMin=-0.5
   UE.Listener.Orientation.Azimuth=0
   UE.Listener.Orientation.Elevation=0
   UE.Source.OnPlay=null
   Volume.VolumeMax=3
   Volume.VolumeMin=-0.3
   Wave.Wave Assets []
  }
  nodes [
   InitVariable.WaveAsset()
   n1 = InitVariable.Trigger()
   n2 = TriggerCompare.Clamp:Int32(Compare=^UE.Source.OnPlay, A=Array.Num:WaveAsset:Array(Array=^Wave.Wave Assets).Num) {also {defaults {B=1, Type=3}}}
   n3 = UE.RandomTime(Max=^MaxWaveSwitchDuration, Min=^MinWaveSwitchDuration, Next=@n2.TRUE)
   n4 = VariableDeferredAccessor.Bool(Variable=InitVariable.Bool().Variable)
   n5 = TriggerCompare.Clamp:Bool(Compare=UE.TriggerRepeat(Start=@n3.On Next, Period=@n3.Value).RepeatOut, A=@n4.Value)
   TriggerRouteTriggerRouteBool2_0 = "TriggerRoute.Trigger Route (Bool, 2):Bool"(Set 1=@n5.FALSE, Set 0=@n5.TRUE) {also {defaults {Value 0="<MetaSoundBuilderResult.SUCCEEDED: 0>"}}}
   TriggerAnyTriggerAny2_0 = "TriggerAny.Trigger Any (2)"(In 1=UE.Reroute(Value=UE.Reroute(Value=@n2.FALSE).Value).Value, In 0=@TriggerRouteTriggerRouteBool2_0.On Set)
   ArrayRandomGet_0 = Array.Random Get:WaveAsset:Array(
    In Array=^Wave.Wave Assets
    No Repeats=Array.Num:WaveAsset:Array(Array=^Wave.Wave Assets).Num
    Next=VariableAccessor.Trigger(Variable=VariableMutator.Trigger(Variable=@n1.Variable, Value=@TriggerAnyTriggerAny2_0.Out).Variable).Value
   ) {also {defaults {Enable Shared State="<MetaSoundBuilderResult.SUCCEEDED: 0>"}}}
   n6 = VariableDeferredAccessor.Trigger(Variable=InitVariable.Trigger().Variable)
   TriggerAnyTriggerAny2_1 = "TriggerAny.Trigger Any (2)"(In 0=^UE.Source.OnPlay, In 1=@n6.Value)
   n7 = UE.RandomTime(Max=^Enveloper.Pitch.MaxAttackTime, Min=^Enveloper.Pitch.MinAttackTime, Next=@TriggerAnyTriggerAny2_1.Out)
   n8 = UE.RandomTime(Max=^Enveloper.Pitch.MaxReleaseTime, Min=^Enveloper.Pitch.MinReleaseTime, Next=@n7.On Next)
   n9 = /FortMetasound/Patches/RandomFloatValueByTrigger(TriggerIn=@n8.On Next) {also {defaults {FloatMax=1.2, FloatMin=0.8}}}
   ADEnvelopeADEnvelope_0 = AD Envelope.AD Envelope:Float(Attack Curve=@n9.Output, Decay Curve=@n9.Output, Trigger=@n8.On Next, Decay Time=@n8.Value, Attack Time=@n7.Value)
   n10 = VariableMutator.Float(
    Variable=InitVariable.Float().Variable
    Value=MapRange.MapRange:Float(
     In=@ADEnvelopeADEnvelope_0.Out Envelope
     Out Range B=&RandomFloatValueByTrigger(FloatMax=^Enveloper.Pitch.MaxVolumeRange, FloatMin=^Enveloper.Pitch.MinVolumeRange, TriggerIn=@ADEnvelopeADEnvelope_0.On Trigger).Output
    ).Out Value
   )
   n11 = /FortMetasound/Patches/MSP_RandomPitchTimeVolume(
    "Random Pitch Max (Semitones)"=^Pitch.PitchMax
    "Random Pitch Min (Semitones)"=^Pitch.PitchMin
    "Random Volume Change Max (Decibels)"=^Volume.VolumeMax
    "Random Volume Change Min (Decibels)"=^Volume.VolumeMin
    Trigger=UE.Reroute(Value=@ArrayRandomGet_0.On Next).Value
   )
   n12 = VariableMutator.Float(Variable=InitVariable.Float().Variable, Value=@n11.Random Pitch)
   n13 = /FortMetasound/Patches/MSP_QuadCrossfadeWavePlayer(
    In=@ArrayRandomGet_0.On Next
    Sound.Sound=@ArrayRandomGet_0.Value
    Start Time=&MSP_WaveGetRandomStartTime(Sound=@ArrayRandomGet_0.Value).Value
    Pitch Shift=UE.Multiply:Float(AdditionalOperands=VariableAccessor.Float(Variable=@n10.Variable).Value, PrimaryOperand=VariableAccessor.Float(Variable=@n12.Variable).Value).Out
    Fade.InvertOutput=VariableAccessor.Bool(Variable=VariableMutator.Bool(Value=@TriggerRouteTriggerRouteBool2_0.Value, Variable=@n4.Variable).Variable).Value
   ) {also {defaults {Fade.FadeTime=2.5}}}
   n14 = VariableMutator.Float(Variable=InitVariable.Float().Variable, Value=@n11.Random Volume)
   n15 = VariableDeferredAccessor.Trigger(Variable=InitVariable.Trigger().Variable)
   TriggerAnyTriggerAny2_2 = "TriggerAny.Trigger Any (2)"(In 0=^UE.Source.OnPlay, In 1=@n15.Value)
   n16 = UE.RandomTime(Max=^Enveloper.Volume.MaxAttackTime, Min=^Enveloper.Volume.MinAttackTime, Next=@TriggerAnyTriggerAny2_2.Out)
   n17 = UE.RandomTime(Max=^Enveloper.Volume.MaxReleaseTime, Min=^Enveloper.Volume.MinReleaseTime, Next=@n16.On Next)
   n18 = &RandomFloatValueByTrigger(TriggerIn=@n17.On Next) {also {defaults {FloatMax=1.2, FloatMin=0.8}}}
   ADEnvelopeADEnvelope_1 = AD Envelope.AD Envelope:Float(Attack Curve=@n18.Output, Decay Curve=@n18.Output, Attack Time=@n16.Value, Trigger=@n17.On Next, Decay Time=@n17.Value)
   n19 = VariableMutator.Float(
    Variable=InitVariable.Float().Variable
    Value=MapRange.MapRange:Float(
     In=@ADEnvelopeADEnvelope_1.Out Envelope
     Out Range B=&RandomFloatValueByTrigger(FloatMax=^Enveloper.Volume.MaxVolumeRange, FloatMin=^Enveloper.Volume.MinVolumeRange, TriggerIn=@ADEnvelopeADEnvelope_1.On Trigger).Output
    ) {also {defaults {Out Range A=1}}}.Out Value
   )
   MSP_AzimuthPannerQuad_0 = &MSP_AzimuthPannerQuad(
    Azimuth=^UE.Listener.Orientation.Azimuth
    In Front Left=@n13.FrontLeft
    In Front Right=@n13.FrontRight
    In Side Left=@n13.OutSideLeft
    In Side Right=@n13.OutSideRight
    Gain=UE.Multiply:Float(PrimaryOperand=VariableAccessor.Float(Variable=@n14.Variable).Value, AdditionalOperands=VariableAccessor.Float(Variable=@n19.Variable).Value).Out
   )
   n20 = UE.RandomTime(Next=@ADEnvelopeADEnvelope_0.On Done) {also {defaults {Max=30, Min=10}}}
   TriggerDelay_0 = UE.Trigger Delay(In=@n20.On Next, Delay Time=@n20.Value)
   VariableMutator.Trigger(Value=@TriggerDelay_0.Out, Variable=@n6.Variable)
   n21 = UE.RandomTime(Next=@ADEnvelopeADEnvelope_1.On Done) {also {defaults {Max=10, Min=1}}}
   TriggerDelay_1 = UE.Trigger Delay(In=@n21.On Next, Delay Time=@n21.Value)
   VariableMutator.Trigger(Value=@TriggerDelay_1.Out, Variable=@n15.Variable)
  ]
  links [
   {dst="@UE.OutputFormat.Quad.Audio:0", src=MSP_AzimuthPannerQuad_0.Out Front Left}
   {dst="@UE.OutputFormat.Quad.Audio:1", src=MSP_AzimuthPannerQuad_0.Out Front Right}
   {dst="@UE.OutputFormat.Quad.Audio:4", src=MSP_AzimuthPannerQuad_0.Out Side Left}
   {dst="@UE.OutputFormat.Quad.Audio:5", src=MSP_AzimuthPannerQuad_0.Out Side Right}
  ]
 }]
properties {
 assetImportData=&self:AssetImportData
 attenuationSettings=&Fort_Ambient_Athena_Att
 bLooping=true
 duration=10000
 numChannels=4
 outputFormat=Quad
 sampleRate=48000
 soundClassObject=&Fort_Ambient_Outdoors_Athena
 volume=0.501
 preEffectBusSends [{
   audioBus=null
   maxSendDistance=1000
   minSendDistance=100
   minSendLevel=0
   sendLevel=1
   soundSourceBus=&AmbientAudio_Exterior_SourceBus
  }]
}
