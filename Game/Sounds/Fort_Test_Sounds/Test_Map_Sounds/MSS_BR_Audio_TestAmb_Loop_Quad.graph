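# Quad MetaSound Source preset of MSS_AmbientLoop_Quad that provides a looping outdoor ambience test sound using a single tropical daytime wave.
# Its inputs set random-range bounds for wave switch timing, pitch, volume and slow attack/release envelopers, with the graph logic inherited from the parent.
# Output uses the Athena ambient attenuation and outdoor sound class, with a pre-effect send to the exterior ambient source bus.
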MetaSoundSource /Game/Sounds/Fort_Test_Sounds/Test_Map_Sounds/MSS_BR_Audio_TestAmb_Loop_Quad
preset_parent=/FortMetasound/Ambient/MSS_AmbientLoop_Quad
inputs [
 Enveloper.Pitch.MaxAttackTime {type=Time}
 Enveloper.Pitch.MaxReleaseTime {type=Time}
 Enveloper.Pitch.MaxVolumeRange {type=Float}
 Enveloper.Pitch.MinAttackTime {type=Time}
 Enveloper.Pitch.MinReleaseTime {type=Time}
 Enveloper.Pitch.MinVolumeRange {type=Float}
 Enveloper.Volume.MaxAttackTime {type=Time}
 Enveloper.Volume.MaxReleaseTime {type=Time}
 Enveloper.Volume.MaxVolumeRange {type=Float}
 Enveloper.Volume.MinAttackTime {type=Time}
 Enveloper.Volume.MinReleaseTime {type=Time}
 Enveloper.Volume.MinVolumeRange {type=Float}
 MaxWaveSwitchDuration {type=Time}
 MinWaveSwitchDuration {type=Time}
 Pitch.PitchMax {type=Float}
 Pitch.PitchMin {type=Float}
 UE.Listener.Orientation.Azimuth {type=Float}
 UE.Listener.Orientation.Elevation {type=Float}
 UE.Source.OnPlay {type=Trigger}
 Volume.VolumeMax {type=Float}
 Volume.VolumeMin {type=Float}
 Wave.Wave Assets {type=WaveAsset:Array}
]
interfaces [UE.Listener.Orientation:1.0, UE.OutputFormat.Quad:1.0, UE.Source:1.1]
outputs [
 "UE.OutputFormat.Quad.Audio:0" {type=Audio}
 "UE.OutputFormat.Quad.Audio:1" {type=Audio}
 "UE.OutputFormat.Quad.Audio:4" {type=Audio}
 "UE.OutputFormat.Quad.Audio:5" {type=Audio}
]
pages [Default {input_defaults {
   Enveloper.Pitch.MaxAttackTime=15
   Enveloper.Pitch.MaxReleaseTime=15
   Enveloper.Pitch.MaxVolumeRange=0.5
   Enveloper.Pitch.MinAttackTime=10
   Enveloper.Pitch.MinReleaseTime=10
   Enveloper.Pitch.MinVolumeRange=-0.5
   Enveloper.Volume.MaxAttackTime=10
   Enveloper.Volume.MaxReleaseTime=15
   Enveloper.Volume.MaxVolumeRange=1.25
   Enveloper.Volume.MinAttackTime=5
   Enveloper.Volume.MinReleaseTime=10
   Enveloper.Volume.MinVolumeRange=0.6
   MaxWaveSwitchDuration=120
   MinWaveSwitchDuration=60
   Pitch.PitchMax=0.5
   Pitch.PitchMin=-0.5
   UE.Listener.Orientation.Azimuth=0
   UE.Listener.Orientation.Elevation=0
   UE.Source.OnPlay=null
   Volume.VolumeMax=3
   Volume.VolumeMin=-0.3
   Wave.Wave Assets [/Game/Sounds/Artemis/Biomes/Tropical/Ambience_Tropical_Day_Loop_Quad]
  }}]
properties {
 assetImportData=&self:AssetImportData
 attenuationSettings=&Fort_Ambient_Athena_Att
 bLooping=true
 duration=10000
 numChannels=4
 outputFormat=Quad
 sampleRate=48000
 soundClassObject=&Fort_Ambient_Outdoors_Athena
 volume=0.501
 preEffectBusSends [{
   audioBus=null
   maxSendDistance=1000
   minSendDistance=100
   minSendLevel=0
   sendLevel=1
   soundSourceBus=&AmbientAudio_Exterior_SourceBus
  }]
}
